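# Showdown detail view widget that warns players a tournament locks them to their current region once they play a match, and asks them to confirm that region.
# On Refresh Data, it joins the tournament's two title lines in uppercase and uppercases the current region name.
# Both values fill a localized format string set on the root rich text block, using inline TournamentStyles tags to highlight the tournament and region.
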WidgetBlueprint /FortniteCoreUI/UI/Events/Notifications/EventRegionLockMessage
graphs [EventGraph {nodes [
   Event Refresh Data() {is=title K2Node_Event}
   n1 = GetCurrentRegionName!(self:Fort Showdown Detail View Object Reference) -> (ReturnValue:Text)
   n2 = Break Fort Tournament Display Info!(FortTournamentDisplayInfo:"Fort Tournament Display Info Structure (by ref)"=Get TournamentDisplayInfo() -> (TournamentDisplayInfo:Fort Tournament Display Info Structure)) -> (
    TitleLine1:Text
    TitleLine2:Text
    ScheduleInfo:Text
    FlavorDescription:Text
    DetailsDescription:Text
    ShortFormatTitle:Text
    LongFormatTitle:Text
    BackgroundTitle:Text
    PinScoreRequirement:int
    PinEarnedText:Text
    BaseColor:Linear Color Structure
    PrimaryColor:Linear Color Structure
    SecondaryColor:Linear Color Structure
    HighlightColor:Linear Color Structure
    TitleColor:Linear Color Structure
    ShadowColor:Linear Color Structure
    BackgroundLeftColor:Linear Color Structure
    BackgroundRightColor:Linear Color Structure
    BackgroundTextColor:Linear Color Structure
    PosterFadeColor:Linear Color Structure
    AlertText:Text
    SeriesPointLeaderboardName:Text
    AlertType:EFortTournamentAlertType Enum
    RoundDisplayNames:Array of Texts
   ) {is=title K2Node_BreakStruct}
   n3 = TextToUpper!(InText:"Text (by ref)"=@n2) -> (ReturnValue:Text)
   n4 = TextToUpper(InText=@n2.TitleLine2)
   > SetText(
    self=Get RichTextBlock_RegionLockMessage() -> (RichTextBlock_RegionLockMessage:Rich Text Block Object Reference)
    InText=Format Text!(
     Format="LOCTEXT(\"<RegionLockMessage>Once you play a match in </><RegionLockHighlight>{Tournament Name}</><RegionLockMessage> in your current region (</><RegionLockHighlight>{Region Name}</><RegionLockMessage>), you will be unable to participate in any other region.</>\\r\\n\\r\\n<RegionLockMessage>Are you sure </><RegionLockHighlight>{Region Name}</><RegionLockMessage> is the region you wish to compete in for </><RegionLockHighlight>{Tournament Name}</><RegionLockMessage>?</>\")"
     Tournament Name:Text=Format Text!(Format="LOCTEXT(\"{0} {1}\")", 0:Text=@n3, 1:Text=@n4) -> (Result)
     Region Name:Text=TextToUpper(InText=@n1)
    ) -> (Result)
   )
  ]}]
variables []
widgets {
 info {inheritedWidgetCount=6, parentClass=/Script/FortniteUI.FortShowdownDetailView, rootWidgetClass=/Script/UMG.RichTextBlock, widgetCount=7}
 widgets [
  RichTextBlock_RegionLockMessage {
   widgetClassPath=&RichTextBlock
   bIsVariable=true
   widget=RichTextBlock_RegionLockMessage
   properties {
    autoWrapText=true
    text="<RegionLockMessage>Once you play a match in </><RegionLockHighlight>{Tournament Name}</><RegionLockMessage> in your current region (</><RegionLockHighlight>{Region Name}</><RegionLockMessage>), you will be unable to participate in any other region.</>\r\n\r\n<RegionLockMessage>Are you sure </><RegionLockHighlight>{Region Name}</><RegionLockMessage> is the region you wish to compete in for </><RegionLockHighlight>{Tournament Name}</><RegionLockMessage>?</>"
    textStyleSet=/FortniteCoreUI/Frontend/Showdown/TournamentStyles
   }
  }
  Text_TimeRemaining {widgetClassPath=&CommonTextBlock, bInherited=true, widget=null}
  Text_DebugId {widgetClassPath=&TextBlock, bInherited=true, widget=null, properties=~}
  Image_PosterFront {widgetClassPath=/Script/EpicCMSUIFramework.EpicCMSImage, bInherited=true, widget=null}
  Image_PosterBack {widgetClassPath=&EpicCMSImage, bInherited=true, widget=null}
  Image_LoadingScreen {widgetClassPath=&EpicCMSImage, bInherited=true, widget=null}
  Image_PlaylistTile {widgetClassPath=&EpicCMSImage, bInherited=true, widget=null}
 ]
}
